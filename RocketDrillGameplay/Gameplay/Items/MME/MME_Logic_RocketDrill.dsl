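; Movement mode extension logic for the Rocket Drill item, driving the pawn while it rides the drill.
; Gameplay events trigger launches from the ground, and PhysUpdate computes velocity per launching, airborne or grounded state, using floor checks for landing and falling.
; Granted gameplay tags for startup, boosting and obstacle blocking shape speed and gravity.
; Each update also drains weapon durability, animates the mesh, toggles gameplay cues and prints velocity debug strings.
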(event EventOnExtensionPush
  (params (RuntimeContext :type MMERuntime%20Context%20Structure%20%28by%20ref%29))
  (bind _runtimedata (Variables|Default|GetRuntimeData))
  (|CallParentFunction RuntimeContext)
  (bind _returnvalue (Class|FortMovementModeExtBlueprintLibrary|GetRuntimeDatafromContext RuntimeContext "/RocketDrillGameplay/Gameplay/Items/MME/MME_Runtime_RocketDrill.MME_Runtime_RocketDrill_C"))
  (Variables|Default|SetRuntimeData _returnvalue)
  (bind _asb_weapon_rocket_drill (Utilities|Casting|CastToB_Weapon_RocketDrill (Variables|/Script/FortniteGame.FortPawn|GetCurrentWeapon (Variables|FortMovementMode_BaseExtLogic|GetOwnerPawn))))
  (Variables|Default|SetWeapon _asb_weapon_rocket_drill)
  (GameplayCue|AddGameplayCue (Variables|FortMovementMode_BaseExtLogic|GetOwnerPawn) (Variables|Default|GetGroundTraversalCue))
  (Ability|Async|WaitGameplayEventToActor (Variables|FortMovementMode_BaseExtLogic|GetOwnerPawn) "(TagName=\"Event.Athena.RocketDrill.JumpInput.Start\")"
    (:then
      (Variables|/RocketDrillGameplay/Gameplay/Items/MME/MME_Runtime_RocketDrill.MME_Runtime_RocketDrill_C|SetLaunchVelocity (* (Math|Vector|GetForwardVector (Pawn|GetDesiredRotation (Pawn|GetController (Variables|FortMovementMode_BaseExtLogic|GetOwnerPawn)))) (ScalableFloat|GetValueatLevel (Variables|Default|GetSF_LaunchSpeed))) _runtimedata)
      (Variables|Default|SetStartTime (Utilities|Time|GetGameTimeinSeconds _runtimedata))
      (Ability|Async|WaitGameplayEventToActor (Variables|FortMovementMode_BaseExtLogic|GetOwnerPawn) "(TagName=\"Event.Athena.RocketDrill.IncomingPeak\")"
        (:EventReceived
          (CallFunction|TriggerLaunch 1.6e+03 7e+01))))
    (:EventReceived
      (CallFunction|TriggerLaunch 2e+03 7.5e+02)
      (Ability|Async|WaitGameplayEventToActor (Variables|FortMovementMode_BaseExtLogic|GetOwnerPawn) "(TagName=\"Event.Athena.RocketDrill.JumpInput.End\")"
        (:EventReceived
          (if (Utilities|Enum|Equal(Enum) (Variables|/RocketDrillGameplay/Gameplay/Items/MME/MME_Runtime_RocketDrill.MME_Runtime_RocketDrill_C|GetState (Variables|Default|GetRuntimeData)) "NewEnumerator0")
            (CallFunction|SetStateinAir true)))))))

(event Custom|TriggerLaunch
  (params (Speed :type Float%20%28single-precision%29) (AddedUpwardSpeed :type Float%20%28double-precision%29))
  (bind (_bwalkablefloor _hit_result) (CallFunction|FindFloor))
  (bind (_bblockinghit _binitialoverlap _time _distance _location _impactpoint _normal) (Collision|BreakHitResult _hit_result))
  (switch Utilities|FlowControl|Switch|SwitchonRocketDrill_State (Variables|/RocketDrillGameplay/Gameplay/Items/MME/MME_Runtime_RocketDrill.MME_Runtime_RocketDrill_C|GetState (Variables|Default|GetRuntimeData))
    (:NewEnumerator3
      (CallFunction|SetStateLaunching)
      (Variables|/RocketDrillGameplay/Gameplay/Items/MME/MME_Runtime_RocketDrill.MME_Runtime_RocketDrill_C|SetLaunchVelocity (Math|Vector|MakeVector (.x (* (Math|Vector|Normalize (Math|Vector|ProjectVectorontoPlane (Math|Vector|Normalize (Variables|Default|GetActualVelocity)) _normal)) Speed)) (.y (* (Math|Vector|Normalize (Math|Vector|ProjectVectorontoPlane (Math|Vector|Normalize (Variables|Default|GetActualVelocity)) _normal)) Speed)) (+ (Math|Float|Clamp(Float) (.z (* (Math|Vector|Normalize (Math|Vector|ProjectVectorontoPlane (Math|Vector|Normalize (Variables|Default|GetActualVelocity)) _normal)) Speed)) 0.0 1.5e+03) AddedUpwardSpeed)) (Variables|Default|GetRuntimeData)))))

(event EventOnExtensionEnd
  (params (RuntimeContext :type MMERuntime%20Context%20Structure%20%28by%20ref%29))
  (bind _ownerpawn (Variables|FortMovementMode_BaseExtLogic|GetOwnerPawn))
  (|CallParentFunction RuntimeContext)
  (Transformation|SetRelativeLocationAndRotation (Variables|/Script/Engine.Character|GetMesh (Variables|FortMovementMode_BaseExtLogic|GetOwnerPawn)) "0.000000,0.000000,-70.000000" "0.000000,-90.000000,0.000000" false true)
  (Components|DestroyComponent (Variables|Default|GetDirtFX) 0)
  (Class|BWeaponRocketDrill|StopRidingRocketDrill (Variables|Default|GetWeapon))
  (GameplayCue|RemoveGameplayCueOnActor(Looping) (Variables|FortMovementMode_BaseExtLogic|GetOwnerPawn) (Variables|Default|GetGC_DrillingOnGround) (Ability|GameplayCue|MakeGameplayCueParameters 0.0))
  (GameplayCue|RemoveGameplayCue _ownerpawn (Variables|Default|GetAirTraversalCue))
  (GameplayCue|RemoveGameplayCue _ownerpawn (Variables|Default|GetGroundTraversalCue)))

(event Custom|SetStateLaunching
  (bind _runtimedata (Variables|Default|GetRuntimeData))
  (Variables|/RocketDrillGameplay/Gameplay/Items/MME/MME_Runtime_RocketDrill.MME_Runtime_RocketDrill_C|SetState "NewEnumerator0" (Variables|Default|GetRuntimeData))
  (Variables|/RocketDrillGameplay/Gameplay/Items/MME/MME_Runtime_RocketDrill.MME_Runtime_RocketDrill_C|SetAnim_GoalMeshPosition "0, 0, 0" _runtimedata)
  (Variables|/RocketDrillGameplay/Gameplay/Items/MME/MME_Runtime_RocketDrill.MME_Runtime_RocketDrill_C|SetTimeLaunching 0.0 _runtimedata)
  (Variables|Default|SetRecentAngleChange 0.0)
  (Variables|FortMovementMode_BPExtLogic|SetbMoveFollowFloor false)
  (Variables|FortMovementMode_BPExtLogic|SetbIsMovingOnGround false)
  (GameplayCue|AddGameplayCue (Variables|FortMovementMode_BaseExtLogic|GetOwnerPawn) (Variables|Default|GetAirTraversalCue))
  (GameplayCue|RemoveGameplayCue (Variables|FortMovementMode_BaseExtLogic|GetOwnerPawn) (Variables|Default|GetGroundTraversalCue)))

(event Custom|SetStateInAir
  (params (bFromLaunch :type Boolean))
  (Variables|/RocketDrillGameplay/Gameplay/Items/MME/MME_Runtime_RocketDrill.MME_Runtime_RocketDrill_C|SetState "NewEnumerator1" (Variables|Default|GetRuntimeData))
  (if (not bFromLaunch)
    (GameplayCue|AddGameplayCue (Variables|FortMovementMode_BaseExtLogic|GetOwnerPawn) (Variables|Default|GetAirTraversalCue))
    (GameplayCue|RemoveGameplayCue (Variables|FortMovementMode_BaseExtLogic|GetOwnerPawn) (Variables|Default|GetGroundTraversalCue))))

(event Custom|SetStateGrounded
  (bind _runtimedata (Variables|Default|GetRuntimeData))
  (Variables|/RocketDrillGameplay/Gameplay/Items/MME/MME_Runtime_RocketDrill.MME_Runtime_RocketDrill_C|SetState "NewEnumerator3" (Variables|Default|GetRuntimeData))
  (GameplayCue|AddGameplayCue (Variables|FortMovementMode_BaseExtLogic|GetOwnerPawn) (Variables|Default|GetGroundTraversalCue))
  (GameplayCue|RemoveGameplayCue (Variables|FortMovementMode_BaseExtLogic|GetOwnerPawn) (Variables|Default|GetAirTraversalCue))
  (Variables|/RocketDrillGameplay/Gameplay/Items/MME/MME_Runtime_RocketDrill.MME_Runtime_RocketDrill_C|SetIsPushedDown true _runtimedata)
  (Variables|/RocketDrillGameplay/Gameplay/Items/MME/MME_Runtime_RocketDrill.MME_Runtime_RocketDrill_C|SetPushedDownAmount 1.5 _runtimedata))

(fn PhysUpdate
  (params (RuntimeContext :type MMERuntime%20Context%20Structure%20%28by%20ref%29) (InVelocity :type Vector%20%28by%20ref%29 :default "0, 0, 0") (InAcceleration :type Vector%20%28by%20ref%29 :default "0, 0, 0"))
  (bind _ownerpawn (Variables|FortMovementMode_BaseExtLogic|GetOwnerPawn))
  (bind _mesh (Variables|/Script/Engine.Character|GetMesh _ownerpawn))
  (bind _returnvalue (Transformation|GetWorldLocation _mesh))
  (bind _returnvalue_2 (CallFunction|IsBoosting))
  (bind _returnvalue_3 (Ability|GameplayCue|MakeGameplayCueParameters 0.0))
  (bind _returnvalue_4 (not _returnvalue_2))
  (bind _returnvalue_5 (Ability|GameplayCue|MakeGameplayCueParameters 0.0))
  (bind _runtime (LocalVariables|GetRuntime))
  (bind _state (Variables|/RocketDrillGameplay/Gameplay/Items/MME/MME_Runtime_RocketDrill.MME_Runtime_RocketDrill_C|GetState _runtime))
  (bind _returnvalue_6 (Utilities|Enum|Equal(Enum) _state "NewEnumerator3"))
  (bind _returnvalue_7 (and _returnvalue_6 _returnvalue_4))
  (bind _runtime_2 (LocalVariables|GetRuntime))
  (bind _lastposition (Variables|/RocketDrillGameplay/Gameplay/Items/MME/MME_Runtime_RocketDrill.MME_Runtime_RocketDrill_C|GetLastPosition _runtime_2))
  (bind _returnvalue_8 (- _returnvalue _lastposition))
  (bind _deltatime (LocalVariables|GetDeltaTime))
  (bind _returnvalue_9 (/ _returnvalue_8 _deltatime))
  (bind _runtime_3 (LocalVariables|GetRuntime))
  (bind _returnvalue_10 (Class|FortMovementModeExtBlueprintLibrary|GetRuntimeDatafromContext RuntimeContext "/RocketDrillGameplay/Gameplay/Items/MME/MME_Runtime_RocketDrill.MME_Runtime_RocketDrill_C"))
  (LocalVariables|SetRuntime _returnvalue_10)
  (bind _output_get (LocalVariables|SetDeltaTime (Class|FortMovementModeExtBlueprintLibrary|GetDeltaTimefromContext RuntimeContext (Variables|FortMovementMode_BaseExtLogic|GetbUseDilatedDeltaTime))))
  (FortWeapon|UseWeaponDurability(Scaled) (Variables|Default|GetWeapon) (* _output_get (LocalVariables|GetDurabilityPerSecond)))
  (Utilities|FlowControl|Sequence
    (:then_0
      (switch Utilities|FlowControl|Switch|SwitchonRocketDrill_State (Variables|/RocketDrillGameplay/Gameplay/Items/MME/MME_Runtime_RocketDrill.MME_Runtime_RocketDrill_C|GetState (LocalVariables|GetRuntime))
        (:NewEnumerator0
          (bind _output_get_2 (LocalVariables|SetNewVelocity (Math|Vector|VectorClampSize2D (+ (Utilities|Operators|Multiply (CallFunction|GoalDirection (Variables|GetInAcceleration)) (LocalVariables|GetDeltaTime) (Utilities|SFToFloat (Variables|Default|GetSF_TurnSpeed))) (Variables|/RocketDrillGameplay/Gameplay/Items/MME/MME_Runtime_RocketDrill.MME_Runtime_RocketDrill_C|GetLaunchVelocity (LocalVariables|GetRuntime))) 0.0 (CallFunction|GetCurrentMaxSpeed))))
          (Variables|FortMovementMode_BPExtLogic|SetbMoveFollowFloor false)
          (Variables|FortMovementMode_BPExtLogic|SetbIsMovingOnGround false)
          (Development|PrintString 0 (Utilities|String|ToString(Vector) _output_get_2) true true "(R=0.537702,G=0.000000,B=1.000000,A=1.000000)")
          (bind _output_get_3 (Variables|/RocketDrillGameplay/Gameplay/Items/MME/MME_Runtime_RocketDrill.MME_Runtime_RocketDrill_C|SetTimeLaunching (+ (Variables|/RocketDrillGameplay/Gameplay/Items/MME/MME_Runtime_RocketDrill.MME_Runtime_RocketDrill_C|GetTimeLaunching (LocalVariables|GetRuntime)) (LocalVariables|GetDeltaTime)) (LocalVariables|GetRuntime)))
          (if (> _output_get_3 0.5)
            (CallFunction|SetStateinAir true)))
        (:NewEnumerator1
          (bind _output_get_4 (LocalVariables|SetNewVelocity (Math|Vector|SelectVector (Math|Vector|MakeVector (.x (Math|Vector|VectorClampSize2D (+ (+ (Utilities|Operators|Multiply "0.000000,0.000000,-1.000000" (LocalVariables|GetDeltaTime) (Math|Float|SelectFloat (Utilities|RemapFloat (Variables|/RocketDrillGameplay/Gameplay/Items/MME/MME_Runtime_RocketDrill.MME_Runtime_RocketDrill_C|GetTimeBoosting (LocalVariables|GetRuntime)) "(X=0.000000,Y=1.000000)" "(X=-200.000000,Y=800.000000)" true) (Utilities|SFToFloat (Variables|Default|GetSF_Gravity)) (CallFunction|IsBoosting))) (Utilities|Operators|Multiply (Math|Vector|Normalize2D(Vector) (CallFunction|GoalDirection (Variables|GetInAcceleration))) (LocalVariables|GetDeltaTime) (Utilities|SFToFloat (Variables|Default|GetSF_TurnSpeed)))) (Variables|GetInVelocity)) 0.0 (CallFunction|GetCurrentMaxSpeed))) (.y (Math|Vector|VectorClampSize2D (+ (+ (Utilities|Operators|Multiply "0.000000,0.000000,-1.000000" (LocalVariables|GetDeltaTime) (Math|Float|SelectFloat (Utilities|RemapFloat (Variables|/RocketDrillGameplay/Gameplay/Items/MME/MME_Runtime_RocketDrill.MME_Runtime_RocketDrill_C|GetTimeBoosting (LocalVariables|GetRuntime)) "(X=0.000000,Y=1.000000)" "(X=-200.000000,Y=800.000000)" true) (Utilities|SFToFloat (Variables|Default|GetSF_Gravity)) (CallFunction|IsBoosting))) (Utilities|Operators|Multiply (Math|Vector|Normalize2D(Vector) (CallFunction|GoalDirection (Variables|GetInAcceleration))) (LocalVariables|GetDeltaTime) (Utilities|SFToFloat (Variables|Default|GetSF_TurnSpeed)))) (Variables|GetInVelocity)) 0.0 (CallFunction|GetCurrentMaxSpeed)))) (Math|Vector|VectorClampSize2D (+ (+ (Utilities|Operators|Multiply "0.000000,0.000000,-1.000000" (LocalVariables|GetDeltaTime) (Math|Float|SelectFloat (Utilities|RemapFloat (Variables|/RocketDrillGameplay/Gameplay/Items/MME/MME_Runtime_RocketDrill.MME_Runtime_RocketDrill_C|GetTimeBoosting (LocalVariables|GetRuntime)) "(X=0.000000,Y=1.000000)" "(X=-200.000000,Y=800.000000)" true) (Utilities|SFToFloat (Variables|Default|GetSF_Gravity)) (CallFunction|IsBoosting))) (Utilities|Operators|Multiply (Math|Vector|Normalize2D(Vector) (CallFunction|GoalDirection (Variables|GetInAcceleration))) (LocalVariables|GetDeltaTime) (Utilities|SFToFloat (Variables|Default|GetSF_TurnSpeed)))) (Variables|GetInVelocity)) 0.0 (CallFunction|GetCurrentMaxSpeed)) (CallFunction|IsBlockedbyObstacle))))
          (Development|PrintString 0 (Utilities|String|ToString(Vector) _output_get_4) true true "(R=1.000000,G=0.011319,B=0.000000,A=1.000000)")
          (if (CallFunction|FindFloor)
            (CallFunction|SetStateGrounded)
            (Variables|FortMovementMode_BPExtLogic|SetbMoveFollowFloor true)
            (Variables|FortMovementMode_BPExtLogic|SetbIsMovingOnGround true)))
        (:NewEnumerator3
          (Utilities|FlowControl|Sequence
            (:then_0
              (bind _output_get_5 (LocalVariables|SetNewVelocity (Math|Vector|VectorClampSize2D (+ (Utilities|Operators|Multiply (+ (CallFunction|GoalDirection (Variables|GetInAcceleration)) (Math|Vector|Normalize (Variables|GetInVelocity))) (LocalVariables|GetDeltaTime) (Utilities|SFToFloat (Variables|Default|GetSF_TurnSpeed))) (Variables|GetInVelocity)) 6e+02 (CallFunction|GetCurrentMaxSpeed))))
              (Development|PrintString 0 (Utilities|String|ToString(Vector) _output_get_5) true true "(R=0.000000,G=0.888935,B=1.000000,A=1.000000)")
              (Variables|FortMovementMode_BPExtLogic|SetbMoveFollowFloor true)
              (Variables|FortMovementMode_BPExtLogic|SetbIsMovingOnGround true))
            (:then_1
              (Variables|/RocketDrillGameplay/Gameplay/Items/MME/MME_Runtime_RocketDrill.MME_Runtime_RocketDrill_C|SetAnim_GoalMeshPosition_X (Math|Vector|MakeVector 0.0 0.0 (Math|Float|SelectFloat -3e+01 -115.0 (CallFunction|IsBoosting))) (LocalVariables|GetRuntime))
              (Class|BWeaponRocketDrill|RidingRocketDrill (Variables|Default|GetWeapon)))
            (:then_2
              (if (not (CallFunction|FindFloor))
                (CallFunction|SetStateinAir false)
                (Variables|FortMovementMode_BPExtLogic|SetbMoveFollowFloor false)
                (Variables|FortMovementMode_BPExtLogic|SetbIsMovingOnGround false)))))
        (:NewEnumerator8
          (Variables|/RocketDrillGameplay/Gameplay/Items/MME/MME_Runtime_RocketDrill.MME_Runtime_RocketDrill_C|SetState "NewEnumerator8" (LocalVariables|GetRuntime)))))
    (:then_1
      (Utilities|FlowControl|Sequence
        (:then_0
          (if (Variables|/RocketDrillGameplay/Gameplay/Items/MME/MME_Runtime_RocketDrill.MME_Runtime_RocketDrill_C|GetIsPushedDown (LocalVariables|GetRuntime))
            (Variables|/RocketDrillGameplay/Gameplay/Items/MME/MME_Runtime_RocketDrill.MME_Runtime_RocketDrill_C|SetAnim_GoalMeshPosition_X (Math|Vector|MakeVector 0.0 0.0 (Math|Float|SelectFloat -3e+01 (Utilities|RemapFloat (Variables|/RocketDrillGameplay/Gameplay/Items/MME/MME_Runtime_RocketDrill.MME_Runtime_RocketDrill_C|GetPushedDownAmount (LocalVariables|GetRuntime)) "(X=0.000000,Y=1.000000)" "(X=-110.000000,Y=-140.000000)" true) (CallFunction|IsBoosting))) (LocalVariables|GetRuntime))
            (bind _output_get_6 (Variables|/RocketDrillGameplay/Gameplay/Items/MME/MME_Runtime_RocketDrill.MME_Runtime_RocketDrill_C|SetPushedDownAmount (- (Variables|/RocketDrillGameplay/Gameplay/Items/MME/MME_Runtime_RocketDrill.MME_Runtime_RocketDrill_C|GetPushedDownAmount (LocalVariables|GetRuntime)) (LocalVariables|GetDeltaTime)) (LocalVariables|GetRuntime)))
            (if (<= _output_get_6 0)
              (Variables|/RocketDrillGameplay/Gameplay/Items/MME/MME_Runtime_RocketDrill.MME_Runtime_RocketDrill_C|SetIsPushedDown false (LocalVariables|GetRuntime)))))
        (:then_1
          (if (!= (CallFunction|IsBoosting) (Variables|Default|GetAppliedBoostingGC))
            (bind _output_get_7 (Variables|Default|SetAppliedBoostingGC (CallFunction|IsBoosting)))
            (if _output_get_7
              (Class|BWeaponRocketDrill|BoostStateChanged (Variables|Default|GetWeapon) true)
              (GameplayCue|AddGameplayCueOnActor(Looping) (Variables|FortMovementMode_BaseExtLogic|GetOwnerPawn) (Variables|Default|GetGC_DrillingOnGround) _returnvalue_5)
              (else
                (Class|BWeaponRocketDrill|BoostStateChanged (Variables|Default|GetWeapon) false)
                (GameplayCue|RemoveGameplayCueOnActor(Looping) (Variables|FortMovementMode_BaseExtLogic|GetOwnerPawn) (Variables|Default|GetGC_DrillingOnGround) _returnvalue_5)))))
        (:then_2
          (if (CallFunction|IsBoosting)
            (Variables|/RocketDrillGameplay/Gameplay/Items/MME/MME_Runtime_RocketDrill.MME_Runtime_RocketDrill_C|SetTimeBoosting (+ (Variables|/RocketDrillGameplay/Gameplay/Items/MME/MME_Runtime_RocketDrill.MME_Runtime_RocketDrill_C|GetTimeBoosting _runtime_3) (LocalVariables|GetDeltaTime)) _runtime_3)
            (else
              (Variables|/RocketDrillGameplay/Gameplay/Items/MME/MME_Runtime_RocketDrill.MME_Runtime_RocketDrill_C|SetTimeBoosting 0.0 _runtime_3))))
        (:then_3
          (Variables|Default|SetActualVelocity _returnvalue_9)
          (if (not (CallFunction|IsBlockedbyObstacle))
            (Transformation|SetWorldRotation (Variables|/Script/Engine.Character|GetMesh (Variables|FortMovementMode_BaseExtLogic|GetOwnerPawn)) (Math|Interpolation|RInterpTo (Transformation|GetWorldRotation (Variables|/Script/Engine.Character|GetMesh (Variables|FortMovementMode_BaseExtLogic|GetOwnerPawn))) (Math|Rotator|CombineRotators (Math|Rotator|MakeRotfromYZ _returnvalue_9 (Math|Vector|MakeVector 0.0 0.0 1.0)) "0.000000,-10.000000,0.000000") (LocalVariables|GetDeltaTime) 6.0) false false))
          (Variables|/RocketDrillGameplay/Gameplay/Items/MME/MME_Runtime_RocketDrill.MME_Runtime_RocketDrill_C|SetLastPosition (Transformation|GetWorldLocation (Variables|/Script/Engine.Character|GetMesh (Variables|FortMovementMode_BaseExtLogic|GetOwnerPawn))) (LocalVariables|GetRuntime))
          (Transformation|SetRelativeLocation (Variables|/Script/Engine.Character|GetMesh (Variables|FortMovementMode_BaseExtLogic|GetOwnerPawn)) (Math|Interpolation|VInterpTo (Variables|/Script/Engine.SceneComponent|GetRelativeLocation (Variables|/Script/Engine.Character|GetMesh (Variables|FortMovementMode_BaseExtLogic|GetOwnerPawn))) (Variables|/RocketDrillGameplay/Gameplay/Items/MME/MME_Runtime_RocketDrill.MME_Runtime_RocketDrill_C|GetAnim_GoalMeshPosition (LocalVariables|GetRuntime)) (LocalVariables|GetDeltaTime) (Math|Float|SelectFloat 5.0 5.0 (Variables|/RocketDrillGameplay/Gameplay/Items/MME/MME_Runtime_RocketDrill.MME_Runtime_RocketDrill_C|GetIsPushedDown (LocalVariables|GetRuntime)))) false false))))
    (:then_2
      (if (!= _returnvalue_7 (Variables|Default|GetSpawnedDirtFX))
        (bind _output_get_8 (Variables|Default|SetSpawnedDirtFX _returnvalue_7))
        (if _output_get_8
          (GameplayCue|AddGameplayCueOnActor(Looping) (Variables|FortMovementMode_BaseExtLogic|GetOwnerPawn) (Variables|Default|GetGC_DrillingOnGround) _returnvalue_3)
          (else
            (GameplayCue|RemoveGameplayCueOnActor(Looping) (Variables|FortMovementMode_BaseExtLogic|GetOwnerPawn) (Variables|Default|GetGC_DrillingOnGround) _returnvalue_3)))))
    (:then_3
      (return (LocalVariables|GetNewVelocity) (Variables|GetInAcceleration)))))

(fn UpdateBeforeCharacterMovement
  (params (RuntimeContext :type MMERuntime%20Context%20Structure%20%28by%20ref%29))
  (if (GameplayTags|HasMatchingGameplayTag (Variables|FortMovementMode_BaseExtLogic|GetOwnerPawn) "(TagName=\"Gameplay.Action.Player.RocketDrill.Active\")")
    (if (Utilities|Enum|Equal(Enum) (Variables|/RocketDrillGameplay/Gameplay/Items/MME/MME_Runtime_RocketDrill.MME_Runtime_RocketDrill_C|GetState (Variables|Default|GetRuntimeData)) "NewEnumerator8")
      (Ability|SendGameplayEventtoActor (Variables|FortMovementMode_BaseExtLogic|GetOwnerPawn) "(TagName=\"Event.Athena.RocketDrill.Stop\")")
      (return "Ended")
      (else
        (return "Active")))
    (else
      (return "Ended"))))

(fn ChangeState
  (params (State :type RocketDrill_State%20Enum :default "NewEnumerator0"))
  (bind _runtimedata (Variables|Default|GetRuntimeData))
  (bind _output_get (Variables|/RocketDrillGameplay/Gameplay/Items/MME/MME_Runtime_RocketDrill.MME_Runtime_RocketDrill_C|SetState State (Variables|Default|GetRuntimeData)))
  (switch Utilities|FlowControl|Switch|SwitchonRocketDrill_State _output_get
    (:NewEnumerator0
      (Variables|/RocketDrillGameplay/Gameplay/Items/MME/MME_Runtime_RocketDrill.MME_Runtime_RocketDrill_C|SetAnim_GoalMeshPosition "0, 0, 0" _runtimedata)
      (Variables|/RocketDrillGameplay/Gameplay/Items/MME/MME_Runtime_RocketDrill.MME_Runtime_RocketDrill_C|SetTimeLaunching 0.0 _runtimedata)
      (Variables|Default|SetRecentAngleChange 0.0)
      (Variables|FortMovementMode_BPExtLogic|SetbMoveFollowFloor false)
      (Variables|FortMovementMode_BPExtLogic|SetbIsMovingOnGround false))
    (:NewEnumerator1
      (Variables|FortMovementMode_BPExtLogic|SetbMoveFollowFloor false)
      (Variables|FortMovementMode_BPExtLogic|SetbIsMovingOnGround false)
      (Variables|Default|SetRecentAngleChange 0.0))
    (:NewEnumerator3
      (Variables|FortMovementMode_BPExtLogic|SetbMoveFollowFloor true)
      (Variables|FortMovementMode_BPExtLogic|SetbIsMovingOnGround true))))

(fn FindFloor
  (bind (_bblockinghit _bwalkablefloor _blinetrace _floordist _linedist _hitresult) (Utilities|Struct|BreakFindFloorResult (Pawn|Components|CharacterMovement|FindFloor (Variables|/Script/Engine.Character|GetCharacterMovement (Variables|FortMovementMode_BaseExtLogic|GetOwnerPawn)) (Transformation|GetWorldLocation (Variables|/Script/Engine.Character|GetCapsuleComponent (Variables|FortMovementMode_BaseExtLogic|GetOwnerPawn))))))
  (return _bwalkablefloor _hitresult))

(fn GetCurrentMaxSpeed
  (return (Math|Float|SelectFloat 0.0 (Math|Float|SelectFloat 8e+02 (Math|Float|SelectFloat 2.5e+03 1.3e+03 (CallFunction|IsBoosting)) (CallFunction|IsBlockedbyObstacle)) (GameplayTags|HasMatchingGameplayTag (Variables|FortMovementMode_BaseExtLogic|GetOwnerPawn) "(TagName=\"Granted.Athena.RocketDrill.Startup\")"))))

(fn GoalDirection
  (params (Acceleration :type Vector))
  (return (Math|Vector|Normalize (Variables|GetAcceleration))))

(fn IsBlockedByObstacle
  (return (GameplayTags|HasMatchingGameplayTag (Variables|FortMovementMode_BaseExtLogic|GetOwnerPawn) "(TagName=\"Granted.Athena.RocketDrill.BlockedByObstacle\")")))

(fn IsBoosting
  (return (GameplayTags|HasMatchingGameplayTag (Variables|FortMovementMode_BaseExtLogic|GetOwnerPawn) "(TagName=\"Granted.Athena.RocketDrill.Boosting\")")))

(fn IsJumping
  (return (GameplayTags|HasMatchingGameplayTag (Variables|FortMovementMode_BaseExtLogic|GetOwnerPawn) "(TagName=\"Granted.Athena.RocketDrill.Jumping\")")))
