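; Movement mode extension logic for the jetpack version of the Rocket Drill item, handling what happens when the movement extension ends.
; On extension end, it calls the parent handler, then checks whether the owning pawn carries the LandingWithImpact gameplay tag.
; If that tag is present, it zeroes the movement component's velocity, stopping the pawn dead on an impact landing.
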(event EventOnExtensionEnd
  (params (RuntimeContext :type MMERuntime%20Context%20Structure%20%28by%20ref%29))
  (|CallParentFunction RuntimeContext)
  (if (GameplayTags|HasMatchingGameplayTag (Variables|FortMovementMode_BaseExtLogic|GetOwnerPawn) "(TagName=\"Granted.Athena.RocketDrill.LandingWithImpact\")")
    (Variables|/Script/Engine.MovementComponent|SetVelocity "0, 0, 0" (Variables|FortMovementMode_BaseExtLogic|GetOwnerMovementComponent))))
